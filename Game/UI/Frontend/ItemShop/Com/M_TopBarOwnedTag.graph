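# UI translucent material, by its name the "owned" tag in the item shop top bar, that procedurally masks a slightly skewed four-sided shape.
# Each side rotates the UVs with CustomRotator, then SmoothSteps one axis at an edge parameter for a soft border.
# Max merges the four outside masks, and the inverse scales the RGBA parameter's alpha.
# Top and bottom rotation pivots put their edge value in X, not Y.
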Material /Game/UI/Frontend/ItemShop/Com/M_TopBarOwnedTag
graph {
 properties {
  bUsedWithStaticMesh=false
  blendMode=BLEND_Translucent
  materialDomain=MD_UI
  pixelDepthOffsetMode=PDOM_Legacy
  usedShadingModels=Default Lit
 }
 nodes [
  n1 = TextureCoordinate()
  n2 = ScalarParameter() {defaultValue=0.14, group=Inner, parameterName=LeftEdge_1}
  n3 = AppendVector(A=@n2, B=Constant() {r=0.5})
  n4 = ScalarParameter() {defaultValue=0.014, group=Inner, parameterName=LeftRotation_1, sliderMax=0.25, sliderMin=-0.25}
  n5 = MaterialFunctionCall(
   Alpha=ComponentMask(None=MaterialFunctionCall(UVs=@n1, Rotation Center=@n3, "Rotation Angle (0-1)"=@n4) {materialFunction=&CustomRotator}.Rotated Values) {r=true}
   Min=@n2
   Max=Add(A=@n2, B=ScalarParameter() {defaultValue=0.005, group=Inner, parameterName=EdgeSoftness})
  ) {materialFunction=&SmoothStep}
  n6 = OneMinus(None=ScalarParameter() {defaultValue=0.215, group=Inner, parameterName=RightEdge_1})
  n7 = Constant() {r=0.5}
  n8 = ComponentMask(None=MaterialFunctionCall(
    UVs=TextureCoordinate()
    Rotation Center=AppendVector(A=@n6, B=@n7)
    "Rotation Angle (0-1)"=ScalarParameter() {defaultValue=-0.0047, group=Inner, parameterName=RightRotation_1, sliderMax=0.25, sliderMin=-0.25}
   ) {materialFunction=&CustomRotator}.Rotated Values) {r=true}
  n9 = Add(A=@n6, B=ScalarParameter() {defaultValue=0.005, group=Inner, parameterName=EdgeSoftness})
  n10 = OneMinus(None=ScalarParameter() {defaultValue=0.23, group=Inner, parameterName=BottomEdge_1})
  n11 = MaterialFunctionCall(
   UVs=TextureCoordinate()
   Rotation Center=AppendVector(A=@n10, B=Constant() {r=0.5})
   "Rotation Angle (0-1)"=ScalarParameter() {defaultValue=0.0095, group=Inner, parameterName=BottomRotation_1, sliderMax=0.25, sliderMin=-0.25}
  ) {materialFunction=&CustomRotator}
  n12 = ScalarParameter() {defaultValue=0.005, group=Inner, parameterName=EdgeSoftness}
  n13 = Max(
   A=Max(
    A=OneMinus(None=@n5.Result)
    B=MaterialFunctionCall(Alpha=@n8, Min=@n6, Max=@n9) {materialFunction=&SmoothStep}.Result
   )
   B=MaterialFunctionCall(Alpha=ComponentMask(None=@n11.Rotated Values) {g=true}, Min=@n10, Max=Add(A=@n10, B=@n12)) {materialFunction=&SmoothStep}.Result
  )
  n14 = ScalarParameter() {defaultValue=0.253, group=Inner, parameterName=TopEdge_1}
  n15 = MaterialFunctionCall(
   UVs=TextureCoordinate()
   Rotation Center=AppendVector(A=@n14, B=Constant() {r=0.5})
   "Rotation Angle (0-1)"=ScalarParameter() {defaultValue=-0.0095, group=Inner, parameterName=TopRotation_1, sliderMax=0.25, sliderMin=-0.25}
  ) {materialFunction=&CustomRotator}
  n16 = ScalarParameter() {defaultValue=0.005, group=Inner, parameterName=EdgeSoftness}
  n17 = OneMinus(None=MaterialFunctionCall(Alpha=ComponentMask(None=@n15.Rotated Values) {g=true}, Min=@n14, Max=Add(A=@n14, B=@n16)) {materialFunction=&SmoothStep}.Result)
  LinearInterpolate(B=VectorParameter() {parameterName=RGBA, defaultValue=<rgba 0 0.4219 0.0328 0.5>}.A, Alpha=OneMinus(None=Max(A=@n13, B=@n17)))
 ]
}
