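# Additive UI material for the item shop frontend that draws an animated, tinted triangle pattern with a soft radial falloff.
# A rotated, slowly panning triangle texture's red channel offsets a time-driven sine, so differently shaded triangles pulse out of phase.
# A symmetric SmoothStep shapes the pulse, which is clamped and multiplied by Color.
# A separate radial mask on UV1, set by center, radius and power, is scaled by Alpha.
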Material /Game/UI/Frontend/ItemShop/Com/vfx/M_Pattern_Triangles
graph {
 properties {bUsedWithStaticMesh=false, blendMode=BLEND_Additive, materialDomain=MD_UI, pixelDepthOffsetMode=PDOM_Legacy, usedShadingModels=Default Lit}
 nodes [
  n1 = Panner(
   Coordinate=Rotator(Coordinate=TextureCoordinate(), Time=ScalarParameter() {parameterName=Rotation})
   Speed=AppendVector(
    A=ScalarParameter() {defaultValue=0.001, parameterName=Panner Speed X}
    B=ScalarParameter() {defaultValue=0.002, parameterName=Panner Speed Y}
   )
  ) {speedX=0.01, speedY=-0.002}
  n2 = Time()
  n3 = ScalarParameter() {defaultValue=1, parameterName=Speed}
  n4 = Sine(None=Add(
    A=TextureSampleParameter2D(UVs=@n1) {parameterName=Param, samplerType=SAMPLERTYPE_LinearColor, texture=&/T_Pattern_Triangles}.R
    B=Multiply(A=@n2, B=@n3)
   ))
  n5 = ScalarParameter() {defaultValue=1, parameterName=SmoothStep, sliderMax=1}
  n6 = Multiply(A=@n5) {constB=-1}
  Multiply(
   A=Clamp(None=MaterialFunctionCall(Alpha=@n4, Min=@n6, Max=@n5) {materialFunction=&SmoothStep}.Result)
   B=VectorParameter() {parameterName=Color, defaultValue=<rgba 0 0.7538 1 0>}.RGB
  )
  n7 = TextureCoordinate() {coordinateIndex=1}
  n8 = AppendVector(A=ScalarParameter() {defaultValue=-0.5, parameterName=Center X}, B=ScalarParameter() {defaultValue=-0.5, parameterName=Center Y})
  n9 = MaterialFunctionCall(Vector 2=Multiply(A=Add(A=@n7, B=@n8) {constB=-0.5}, B=ScalarParameter() {defaultValue=2, parameterName=Radius}) {constB=2}) {materialFunction=/Engine/Functions/Engine_MaterialFunctions02/Utility/VectorLength}
  n10 = ScalarParameter() {defaultValue=2, parameterName=Power}
  Multiply(A=ScalarParameter() {defaultValue=0.01, parameterName=Alpha, sliderMax=1}, B=OneMinus(None=Power(Base=@n9.V2 Length, Exp=@n10)))
 ]
}
